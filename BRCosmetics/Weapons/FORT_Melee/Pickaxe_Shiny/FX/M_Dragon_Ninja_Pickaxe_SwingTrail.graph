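# Unlit translucent swing-trail material for the Dragon Ninja pickaxe, usable on beam trails and Niagara ribbons.
# Two smoke textures pan along UV1 on the dynamic Time parameter and combine under a square root to drive both color and opacity.
# Time-of-day-adjusted particle color tints the smoke, then passes through a dynamic hue shift and emissive adjustments.
# Particle alpha, dynamic Alpha, a ramp texture and non-mobile depth fade shape opacity.
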Material /BRCosmetics/Weapons/FORT_Melee/Pickaxe_Shiny/FX/M_Dragon_Ninja_Pickaxe_SwingTrail
graph {
 properties {
  bUsedWithBeamTrails=true
  bUsedWithNiagaraRibbons=true
  blendMode=BLEND_Translucent
  pixelDepthOffsetMode=PDOM_Legacy
  shadingModel=MSM_Unlit
  translucencyLightingMode=TLM_VolumetricDirectional
  usedShadingModels=Unlit
 }
 nodes [
  n1 = TextureCoordinate() {coordinateIndex=1}
  n2 = DynamicParameter() {paramNames [Time, HueShift_OverTime, ColorRange, Alpha]}
  n3 = Panner(
   Coordinate=Multiply(A=@n1, B=Constant2Vector() {g=1, r=0.0618}) {constB=0.618}
   Time=@n2.Time
   Speed=AppendVector(A=ScalarParameter() {defaultValue=0.2, parameterName=Speed2}, B=Constant())
  )
  n4 = Panner(Coordinate=@n1, Time=@n2.Time, Speed=AppendVector(A=ScalarParameter() {defaultValue=0.4, parameterName=Speed1}, B=Constant()))
  n5 = SquareRoot(None=Add(
    A=TextureSample(UVs=@n3) {texture=/Game/Effects/Fort_Effects/Textures/Smoke/T_Smoke_Trail_Soft_2}.R
    B=TextureSample(UVs=@n4) {samplerType=SAMPLERTYPE_Grayscale, texture=&T_Smoke_Trail_Soft}.R
   ))
  n6 = Add(A=Multiply(A=@n5) {constB=0.5}, B=DynamicParameter() {paramNames [Time, HueShift_OverTime, ColorRange, Alpha]}.ColorRange)
  n7 = ParticleColor()
  n8 = MaterialFunctionCall(Color_Morning=@n7.RGB, Color_Day=@n7.RGB, Color_Evening=@n7.RGB, Color_Night=Multiply(A=@n7.RGB) {constB=0.6}) {materialFunction=&Color_Time_of_Day}
  MaterialFunctionCall(Emissive=MaterialFunctionCall(
    Hue Shift Percentage=DynamicParameter() {paramNames [Time, HueShift_OverTime, ColorRange, Alpha]}.HueShift_OverTime
    Texture=Multiply(A=@n6, B=@n8.Result)
   ) {materialFunction=&HueShift}.Result) {materialFunction=&MF_FX_EmissiveAdjustments}
  n9 = DynamicParameter() {paramNames [Time, HueShift_OverTime, ColorRange, Alpha]}
  n10 = ParticleColor()
  n11 = MaterialFunctionCall() {materialFunction=&Swizzle}
  n12 = Multiply(
   A=Multiply(A=Multiply(A=@n9.Alpha, B=@n10.A), B=TextureSample(UVs=@n11.YX) {texture=/Game/Effects/Fort_Effects/Textures/Ramps/T_Ramp_Smooth}.R)
   B=ShadingPathSwitch(Default=DepthFade() {fadeDistanceDefault=25}, Mobile=Constant() {r=1}) {desc=Migrated from FeatureLevelSwitch}
  )
  Reroute(None=Subtract(A=Multiply(A=@n5, B=@n12)) {constB=0.02})
 ]
}
