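# Idle Niagara effect for the Shiny pickaxe cosmetic, emitting short-lived sprite glints on the weapon.
# One looping emitter with randomized duration runs three spawn groups, each front-loaded by a curve over loop age, with two using separate gameplay and front-end rates.
# Spawn-group masks give each group its own placement, with varying color and camera offset.
# A size curve makes each glint pop in and shrink away.
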NiagaraSystem /BRCosmetics/Weapons/FORT_Melee/Pickaxe_Shiny/FX/NS_Pickaxe_Shiny_Idle
dependencies {usedDataInterfaces [&NiagaraDataInterfaceCurve], usedDynamicInputs [], usedModules [&SystemState]}
emitters {Glints_Diamond {
  emitterUpdateScript [
   &EmitterState(
    Life Cycle Mode=::NewEnumerator1
    Loop Duration=&RandomRangeFloat(Minimum=0.5, Maximum=1)
    Recalculate Duration Each Loop=true
    UseLoopDelay=false
   )
   &SpawnRate(
    SpawnRate=&Multiply_Float(
     A=&FN_FrontEndFloatModifier(Gameplay=25, Front End=10)
     B=&FloatFromCurve(FloatCurve=DataInterface() {propertyValues {Curve {keys [<at 0 1>,<at 0.25 0>]}}}, CurveIndex=$Emitter.NormalizedLoopAge)
    )
    Use Spawn Probability=false
   )
   &SpawnRate(
    SpawnGroup=1
    SpawnRate=&Multiply_Float(
     A=1
     B=&FloatFromCurve(FloatCurve=DataInterface() {propertyValues {Curve {keys [<at 0 1>,<at 0.5 0>]}}}, CurveIndex=$Emitter.NormalizedLoopAge)
    )
    Use Spawn Probability=false
   ) {module=SpawnRate001}
   &SpawnRate(
    SpawnGroup=2
    SpawnRate=&Multiply_Float(
     A=&FN_FrontEndFloatModifier(Gameplay=10, Front End=3)
     B=&FloatFromCurve(FloatCurve=DataInterface() {propertyValues {Curve {keys [<at 0 1>,<at 0.25 0>]}}}, CurveIndex=$Emitter.NormalizedLoopAge)
    )
    Use Spawn Probability=false
   ) {module=SpawnRate002}
  ]
  particleSpawnScript [
   &InitializeParticle(
    Color=&MaskLinearColor(
     Mask Source=0
     Mask Parameter=?
     Mask=0
     Masked Linear Color=<rgba 4.5833 4.6325 5 2>
     Passthrough Linear Color=&MaskLinearColor(Mask Source=0, Mask Parameter=?, Mask=1, Masked Linear Color=<rgba 4.583 4.6325 5 2>, Passthrough Linear Color=<rgba 5 2.435 0.318 1>)
    )
    Lifetime=&RandomRangeFloat(Minimum=0.5, Maximum=0.8, Evaluation Type=::NewEnumerator1)
    Sprite Size=&RandomRangeVector2D(Minimum=<10 5>, Maximum=<15 17>, Evaluation Type=::NewEnumerator1)
    Sprite Size Mode=::NewEnumerator4
    UsePositionOffset=false
    Write Material Random=true
   )
   &SystemLocation(
    MaskSpawnGroup=false
    Offset=&MaskVector(
     Mask Source=0
     Mask Parameter=?
     Mask=0
     Masked Vector=&RandomRangeVector(Minimum=<0.65 6 44>, Maximum=<0.65 7 44>, Evaluation Type=::NewEnumerator1)
     Passthrough Vector=&MaskVector(
      Mask Source=0
      Mask Parameter=?
      Mask=1
      Masked Vector=&MakeVector(
       X=0
       Y=&UniformAOrBFloat(A=9.25, B=4.25, Evaluation Type=::NewEnumerator0, Recalculate Random Each Loop=true, Override Seed=false)
       Z=&RandomRangeFloat(Minimum=50.75, Maximum=50.7)
      )
      Passthrough Vector=&MakeVector(
       X=&RandomRangeFloat(Minimum=1, Maximum=2)
       Y=&RandomRangeFloat(Minimum=-25, Maximum=40)
       Z=&RandomRangeFloat(Minimum=38, Maximum=45)
      )
     )
    )
   )
   &ShapeLocation(
    MaskSpawnGroup=false
    Override Seed=false
    Sphere Radius=&MaskFloat(
     Mask Source=0
     Mask Parameter=?
     Mask=0
     Masked Float=6
     Passthrough Float=&MaskFloat(Mask Source=0, Mask Parameter=?, Mask=1, Masked Float=0, Passthrough Float=1)
    )
    Sphere Surface Distribution=1
    UseOwnerScale=true
   )
   /SetVariables() {isSetParameters=true}
  ]
  particleUpdateScript [
   &ParticleState()
   &ScaleSpriteSize(Uniform Curve Sprite Scale=DataInterface() {propertyValues {Curve {keys [
        <at 0 0 RCIM_Cubic>
        {arriveTangent=4.3307, interpMode=RCIM_Cubic, leaveTangent=4.3307, time=0.1484, value=0.2131}
        {arriveTangent=5.1683, interpMode=RCIM_Cubic, leaveTangent=5.1683, tangentMode=RCTM_User, time=0.2193, value=0.9495}
        {arriveTangent=-2.7909, interpMode=RCIM_Cubic, leaveTangent=-2.7909, tangentMode=RCTM_User, time=0.4478, value=0.589}
        <at 1 0 RCIM_Cubic>
       ]}}})
   &CameraOffset(Camera Offset Amount=&MaskFloat(
     Mask Source=0
     Mask Parameter=?
     Mask=0
     Masked Float=&FloatFromCurve(FloatCurve=DataInterface() {propertyValues {Curve {keys [<at 0 2>,<at 1 5>]}}}, Scale Curve=5)
     Passthrough Float=&MaskFloat(
      Mask Source=0
      Mask Parameter=?
      Mask=1
      Masked Float=25
      Passthrough Float=&FloatFromCurve(FloatCurve=DataInterface() {propertyValues {Curve {keys [<at 0 5>,<at 1 15>]}}}, Scale Curve=5)
     )
    ))
  ]
  rendererClasses [&NiagaraSpriteRendererProperties]
  renderers [{rendererClass=&NiagaraSpriteRendererProperties}]
 }}
summary {
 systemName=NS_Pickaxe_Shiny_Idle
 emitters [{emitterName=Glints_Diamond, rendererClasses [&NiagaraSpriteRendererProperties]}]
 userVariables [
  User.bIsFrontEnd {type=NiagaraBool}
  User.bIsFrontEndPreview {type=NiagaraBool}
 ]
}
