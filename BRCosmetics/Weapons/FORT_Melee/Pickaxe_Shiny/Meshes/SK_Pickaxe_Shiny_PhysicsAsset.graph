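# Physics asset for the Shiny pickaxe melee weapon's skeletal mesh.
# It holds a single capsule body on the Melee_TwoHanded_Handle bone, offset and rotated to serve as a coarse collision shape.
# It has no constraints, so it defines no articulated or jointed ragdoll parts.
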PhysicsAsset /BRCosmetics/Weapons/FORT_Melee/Pickaxe_Shiny/Meshes/SK_Pickaxe_Shiny_PhysicsAsset
bodies {Melee_TwoHanded_Handle [{shapeType=Capsule, length=57.3644, radius=38.3534, center=<0.7246 -0.901 23.861>, rotation=<rot -0.0178 -90 -0.0977>}]}
constraints []
